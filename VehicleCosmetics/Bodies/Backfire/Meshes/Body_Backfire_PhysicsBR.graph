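# Physics asset for the Backfire vehicle body cosmetic, with the BR suffix marking a variant of the body's physics.
# It holds a single body on the root bone with no collision shapes listed and no constraints.
# The skeletal mesh therefore has no jointed or per-bone simulated parts defined here.
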PhysicsAsset /VehicleCosmetics/Bodies/Backfire/Meshes/Body_Backfire_PhysicsBR
bodies {root []}
constraints []
